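# Translucent UI material for the MistRaven GrimStrut reticle wings, drawing a mirrored MSDF wing texture with an adjustable X offset.
# The MSDF converts to an SDF, thresholded twice for anti-aliased art and shadow, then multiplied by a rotated horizontal fade for master opacity.
# A Fill parameter remaps a rotated vertical gradient to split fill and background colors.
# Panning Perlin noise builds distorted UVs, weighted toward the left and right edges.
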Material /MistRaven/HUD/GrimStrut/Materials/M_UI_MistRaven_GrimStrut_Reticle_Wings
graph {
 properties {bAutomaticallySetUsageInEditor=false, blendMode=BLEND_Translucent, materialDomain=MD_UI, usedShadingModels=Default Lit}
 nodes [
  Add(A=Multiply(A=NamedRerouteUsage(), B=NamedRerouteUsage()), B=NamedRerouteUsage())
  n1 = Clamp(None=MaterialFunctionCall!(UVs=NamedRerouteUsage()) {materialFunction=/FortUILibrary/MaterialFunctions/Basics/MF_UI_9SliceUVs}.UVs)
  MaterialExpressionScalarParameter_7 = ScalarParameter() {parameterName=WingsXOffset}
  n2 = Convert(None=@MaterialExpressionScalarParameter_7) {
   convertInputs [{
     type=Scalar
     expressionInput {
      expression=&self:MaterialExpressionScalarParameter_7
      inputName=None
      mask=0
      maskA=0
      maskB=0
      maskG=0
      maskR=0
      outputIndex=0
     }
    }]
   convertOutputs [{type=Vector2}]
  }
  n3 = Constant2Vector() {g=1, r=2}
  n4 = MaterialFunctionCall!(MSDF RGB=TextureSample(UVs=MaterialFunctionCall!(UV=@n1, Offset=@n2, Rotation, Scale=@n3, MirrorY) {materialFunction=/FortUILibrary/MaterialFunctions/Basics/MF_UI_Symmetry_2x1_DEPRECATED}.Result) {texture=/MistRaven/HUD/GrimStrut/Textures/T_UI_MSDF_MistRaven_GrimStrut_Wing_New}.RGB) {materialFunction=/FortUILibrary/MaterialFunctions/Shapes/Modifiers/MF_UI_MSDFtoSDF}
  Composite!(SDF=@n4.SDF) {subgraphName=Wings AA'd + Shadow}
  Constant() {r=0.5}
  Constant() {r=0.4}
  Constant() {r=0.85}
  n5 = Time()
  Multiply(A=Sine(None=Multiply(A=@n5) {constB=0.15})) {constB=0.5}
  MaterialExpressionReroute_0 = Reroute(None=@n4.SDF)
  n6 = Constant() {r=0.25}
  n7 = MaterialFunctionCall(DistanceField=@MaterialExpressionReroute_0.SDF, Threshold=@n6) {materialFunction=&MF_UI_AntiAliasing}
  MaterialExpressionReroute_1 = Reroute(None=@n7.Outer)
  NamedRerouteDeclaration(None=@MaterialExpressionReroute_1.AA'd Art) {name=WingsAA'dArt, nodeColor=<rgba 0.4531 1 0 1>}
  MaterialExpressionScalarParameter_8 = ScalarParameter() {parameterName=WingsXOffset}
  n8 = Convert(None=@MaterialExpressionScalarParameter_8) {
   convertInputs [{
     type=Scalar
     expressionInput {
      expression=&self:MaterialExpressionScalarParameter_8
      inputName=None
      mask=0
      maskA=0
      maskB=0
      maskG=0
      maskR=0
      outputIndex=0
     }
    }]
   convertOutputs [{type=Vector2}]
  }
  n9 = ComponentMask(None=MaterialFunctionCall(UV=TextureCoordinate(), Offset=@n8, Rotation=ScalarParameter() {defaultValue=-0.01, parameterName=FillRotation}) {materialFunction=&MF_UI_Symmetry_2x1_DEPRECATED}.Result) {g=true}
  n10 = ScalarParameter() {defaultValue=1, parameterName=Fill}
  n11 = Constant2Vector() {g=1}
  n12 = Constant2Vector() {g=1.3, r=-6.45}
  n13 = Saturate(None=Add(
    A=Multiply(A=@n9) {constB=5}
    B=MaterialFunctionCall(Input=@n10, Input Low=@n11.R, Input High=@n11.R, Target Low=@n12.R, Target High=@n12.R) {materialFunction=&RemapValueRange}.Result
   ) {constB=-1})
  NamedRerouteDeclaration(None=Multiply(A=VectorParameter() {parameterName=FillColor, defaultValue=<rgba 1 0.0048 0.0595 1>}.RGB, B=@n13)) {name=FillColor, nodeColor=<rgba 0 0 0 1>}
  n14 = ComponentMask(None=MaterialFunctionCall() {materialFunction=&MF_UI_9SliceUVs}.UVs) {r=true}
  n15 = Multiply(A=@n14, B=OneMinus(None=@n14))
  n16 = OneMinus(None=Saturate(None=Multiply(A=@n15) {constB=4}))
  n17 = ScalarParameter() {defaultValue=0.06, parameterName=DistortionInfluenceGap}
  n18 = NamedRerouteDeclaration(None=Saturate(None=Subtract(A=@n16, B=@n17))) {name=DistortionInfluenceGap, nodeColor=<rgba 0 0.9766 1 1>}
  n19 = MaterialFunctionCall() {materialFunction=&MF_UI_9SliceUVs}
  n20 = Constant2Vector() {g=1, r=2}
  n21 = ScalarParameter() {defaultValue=0.85, group=Distortion, parameterName=DistortionSpeedX}
  n22 = ScalarParameter() {defaultValue=0.5, group=Distortion, parameterName=DistortionSpeedY}
  n23 = TextureSampleParameter2D(UVs=Panner(
    Coordinate=MaterialFunctionCall(UV=@n19.UVs, Scale=@n20) {materialFunction=&MF_UI_Symmetry_2x1_DEPRECATED}.Result
    Speed=AppendVector(A=@n21, B=@n22)
   ) {speedX=0.4}) {group=Distortion, parameterName=DistortionNoise, texture=/Game/Athena/HUD/XP/Materials/T_UI_PerlinNoiseRift}
  n24 = AppendVector(A=Multiply(B=@n23.G), B=Multiply(A=@n23.R, B=ScalarParameter() {defaultValue=0.75, parameterName=DistortionScale}))
  NamedRerouteDeclaration(None=Add(A=Multiply(A=@n18, B=@n24) {constA=1}, B=TextureCoordinate())) {name=DistortedUVs, nodeColor=<rgba 0.9453 0 1 1>}
  n25 = NamedRerouteUsage()
  NamedRerouteDeclaration(None=Multiply(A=Subtract(A=@n25, B=@n13), B=VectorParameter() {parameterName=FillBackgroundColor, defaultValue=<rgba 0 0 0 1>}.RGB)) {name=FillBackgroundColor, nodeColor=<rgba 0 1 0.9063 1>}
  n26 = MaterialFunctionCall(DistanceField=@MaterialExpressionReroute_0.SDF, Threshold=Add(A=@n6) {constB=-0.1}) {materialFunction=&MF_UI_AntiAliasing}
  MaterialExpressionReroute_2 = Reroute(None=@n26.Outer)
  n27 = ComponentMask(None=MaterialFunctionCall(
    UV=MaterialFunctionCall() {materialFunction=&MF_UI_9SliceUVs}.UVs
    Offset=@n2
    Rotation=ScalarParameter() {defaultValue=0.02, parameterName=FadeRotationMask}
   ) {materialFunction=&MF_UI_Symmetry_2x1_DEPRECATED}.Result) {r=true}
  n28 = ScalarParameter() {defaultValue=4, parameterName=x}
  n29 = Multiply(
   A=LinearInterpolate(
    A=@MaterialExpressionReroute_1.AA'd Art
    B=@MaterialExpressionReroute_2.Shadow
    Alpha=ScalarParameter() {defaultValue=0.5, parameterName=ShadowOpacity}
   )
   B=Subtract(A=Multiply(A=@n27, B=@n28) {constB=2}, B=ScalarParameter() {defaultValue=0.45, parameterName=EdgeSoftnessFalloff})
  )
  NamedRerouteDeclaration(None=Multiply(A=Saturate(None=@n29), B=ScalarParameter() {defaultValue=0.7, parameterName=MasterOpacity})) {name=WingsMasterOpacity, nodeColor=<rgba 1 0.5859 0 1>}
  NamedRerouteUsage()
  PinBase() {reroutePins [SDF {expression=&self:MaterialExpressionReroute_0}]}
  PinBase!(AA'd Art=@n7.Outer, Shadow=@n26.Outer) {reroutePins [
    AA'd Art {expression=&self:MaterialExpressionReroute_1}
    Shadow {expression=&self:MaterialExpressionReroute_2}
   ]}
 ]
}
